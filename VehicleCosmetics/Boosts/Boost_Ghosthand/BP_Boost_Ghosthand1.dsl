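; BP_Boost_Ghosthand1 is a vehicle cosmetic actor for the Ghosthand boost.
; On BeginPlay it passes its skeletal mesh to the vehicle cosmetics function library to set an initial location offset slightly back and down.
; The construction script only calls the parent, and the tick and overlap events are empty.
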(event EventBeginPlay
  ; Set initial mesh location
  (Class|BPVehicleCosmeticsFunctionLibrary|CosmeticActorSetSkeletalAnimLocation self "-25.000000,0.000000,-4.000000" (Variables|Default|GetSkeletalMesh)))

(event Collision|EventActorBeginOverlap
  (params (OtherActor :type Actor%20Object%20Reference)))

(event EventTick
  (params (DeltaSeconds :type Float%20%28single-precision%29)))

(fn ConstructionScript
  (|CallParentFunction))
