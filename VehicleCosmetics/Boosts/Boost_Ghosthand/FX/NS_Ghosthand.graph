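# Niagara system for the Ghosthand vehicle boost cosmetic, a green-by-default, paintable effect placed behind the car from a local offset.
# A ribbon emitter trails backward from socket-filtered, quality-scaled spawning, using two ribbon renderers.
# A sprite emitter packs four layers into spawn groups, switching lifetime, size, velocity, SubUV frame, vortex pull and alpha per group across four renderers via a visibility tag.
# Both emitters take color from painted-variant override modules.
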NiagaraSystem /VehicleCosmetics/Boosts/Boost_Ghosthand/FX/NS_Ghosthand
dependencies {
 usedDataInterfaces [
  &NiagaraDataInterfaceSpriteRendererInfo
  &NiagaraDataInterfaceCurve
  &NiagaraDataInterfaceSkeletalMesh
  &NiagaraDataInterfaceArrayFloat3
 ]
 usedDynamicInputs [&SelectVectorFromArray]
 usedModules [
  &self:SystemSpawnScript.SystemScriptSource.SystemScriptGraph.NiagaraNodeAssignment_0.SetVariables_#1
  &self:SystemSpawnScript.SystemScriptSource.SystemScriptGraph.NiagaraNodeAssignment_1.SetVariables_#2
  &self:SystemSpawnScript.SystemScriptSource.SystemScriptGraph.NiagaraNodeAssignment_2.SetVariables_#3
  &SystemState
 ]
}
emitters {
 NE_VC_SingleBoost_Ribbon_Parent001 {
  emitterUpdateScript [
   &EmitterState()
   &VC_FilteredSocketCount()
   &VC_RateQualityHandler()
   &VC_SpawnRibbon_v1(SpawnRate=$Emitter.Spawn Rate)
  ]
  particleSpawnScript [
   &VC_PaintedVariantsOverride(
    DefaultColor=<rgba 0.1687 0.625 0.3506 1>
    Pink=<rgba 0.5052 0.1579 0.3789 1>
    R1 DefaultColor=<rgba 0.4343 1.5 0.9426 1>
    Use R1 Painted Overrides=true
   )
   &VC_InitializeRibbon_v1(
    Color=$Particles.Color
    Custom Offset=true
    Lifetime=0.35
    Local Vector Offset=&AddVectorToPosition(Position=$User.DTCarOffset, Vector=<-275 0 25>)
    Offset Amount=1
    Ribbon Width=125
   )
   &AddVelocity(Velocity=<-4000 0 0>) {module=AddVelocity001}
  ]
  particleUpdateScript [
   &ParticleState()
   &SolveForcesAndVelocity(Clamp Velocity=false, Limit Acceleration=false)
   &VC_UpdateRibbon_v1()
  ]
  rendererClasses [&NiagaraRibbonRendererProperties]
  renderers [
   {rendererClass=&NiagaraRibbonRendererProperties}
   {rendererClass=&NiagaraRibbonRendererProperties, rendererIndex=1}
  ]
 }
 NE_VC_SingleBoost_Sprite_Parent002 {
  emitterUpdateScript [
   &EmitterState(Life Cycle Mode=::NewEnumerator1, UseLoopDelay=false)
   &SpawnBurst_Instantaneous(Spawn Group=3, Use Spawn Probability=false, UseLoopCountLimit=true)
   &SpawnRate(SpawnGroup=3, Use Spawn Probability=false) {module=SpawnRate003}
   &SpawnRate(SpawnGroup=2, SpawnRate=35, Use Spawn Probability=false) {module=SpawnRate002}
   &SpawnRate(SpawnGroup=1, SpawnRate=250, Use Spawn Probability=false) {module=SpawnRate001}
   &SpawnRate(SpawnRate=15, Use Spawn Probability=false)
  ]
  particleSpawnScript [
   &VC_PaintedVariantsOverride(DefaultColor=<rgba 0.2351 1 0.4669 1>, Pink=<rgba 0.5052 0.1605 0.3799 1>)
   &InitializeParticle(
    Color=$Particles.Color
    Lifetime=&MergeHelper_FloatSwitch(
     0=&RandomRangeFloat(Minimum=0, Maximum=0.3)
     1=&RandomRangeFloat(Minimum=0.3, Maximum=0.3)
     2=0.3
     3=2
     4=0
     Index=&SpawnGroup()
     Num=::NewEnumerator3
    )
    Position Offset=&AddVectorToPosition(
     Position=$User.DTCarOffset
     Vector=&MergeHelper_VectorSwitch(
      0=<-250 0 20>
      1=<-280 0 20>
      2=<-280 0 20>
      3=<-250 0 20>
      4=<0 0 0>
      Index=&SpawnGroup()
      Num=::NewEnumerator3
     )
    )
    Position Offset Coordinate Space=::Local
    Sprite Rotation Mode=::NewEnumerator2
    Sprite Size Mode=::NewEnumerator3
    Uniform Sprite Size=&MergeHelper_FloatSwitch(
     0=&RandomRangeFloat(Minimum=25, Maximum=100)
     1=50
     2=5
     3=40
     4=0
     Index=&SpawnGroup()
     Num=::NewEnumerator3
    )
    UsePositionOffset=true
    Write Material Random=true
   )
   &AddVelocity(Velocity=&MergeHelper_VectorSwitch(0=<0 0 0>, 1=<-1500 0 0>, 2=<0 0 0>, 3=<0 0 0>, 4=<0 0 0>, Index=&SpawnGroup(), Num=::NewEnumerator3))
   &ShapeLocation(
    MaskSpawnGroup=false
    Override Seed=false
    Sphere Radius=&MergeHelper_FloatSwitch(0=20, 1=30, 2=40, 3=0, 4=0, Index=&SpawnGroup(), Num=::NewEnumerator3)
    UseOwnerScale=true
   )
   &AddVelocity(
    Velocity=/Game/Effects/Niagara/DynamicInputs/FN_DirectionalVectorfromMontageIndex(
     Montage Section Index:NiagaraInt32=1
     Montage Index 1:Vector3f=<-500 1 1>
     Montage Index 2:Vector3f=<-500 -1 0.85>
     Montage Index 3:Vector3f=<0 0 0>
     Montage Index 4:Vector3f=<0 -1 -1>
     Invalid Montage Index:Vector3f=<1 0 0>
    )
    Velocity Speed Scale=0
   ) {module=AddVelocity001}
   /SetVariables(Particles.VisibilityTag=&SpawnGroup()) {isSetParameters=true}
   &SubUVAnimation(
    Direct Frame Index:NiagaraFloat=&MergeHelper_FloatSwitch(
     0=&RandomRangeFloat(Minimum=0, Maximum=4)
     1=&RandomRangeFloat(Minimum=0, Maximum=4)
     2=0
     3=0
     4=0
     Index=&SpawnGroup()
     Num=::NewEnumerator3
    )
    SubUV Animation Mode:ENiagara_SubUVLookupModeV2=::NewEnumerator5
   )
  ]
  particleUpdateScript [
   &ParticleState()
   &VortexForce(
    Origin Pull Amount:NiagaraFloat=&MergeHelper_FloatSwitch(0=0, 1=0, 2=300, 3=20000, 4=0, Index=&SpawnGroup(), Num=::NewEnumerator3)
    UseOriginOffset=false
    Vortex Axis=<-280 0 20>
    Vortex Axis Coordinate Space:ENiagaraCoordinateSpace=::Local
    Vortex Force Amount=&MergeHelper_FloatSwitch(0=0, 1=0, 2=0, 3=100, 4=0, Index=&SpawnGroup(), Num=::NewEnumerator3)
   )
   &SolveForcesAndVelocity(Clamp Velocity=false, Limit Acceleration=false)
   &ScaleColor(
    Scale Alpha=&MergeHelper_FloatSwitch(
     0=1
     1=&FloatFromCurve(FloatCurve=DataInterface() {propertyValues {Curve {keys [<at 0 1>,<at 1 0>]}}})
     2=&FloatFromCurve(FloatCurve=DataInterface() {propertyValues {Curve {keys [<at 0.0093 -0.0249>,<at 0.5018 0.9871>,<at 1 0>]}}})
     3=1
     4=0
     Index=&SpawnGroup()
     Num=::NewEnumerator3
    )
    Scale RGB=<2 2 2>
    ScaleA=true
    ScaleRGB=true
   )
  ]
  rendererClasses [&NiagaraSpriteRendererProperties]
  renderers [
   {rendererClass=&NiagaraSpriteRendererProperties}
   {rendererClass=&NiagaraSpriteRendererProperties, rendererIndex=1}
   {rendererClass=&NiagaraSpriteRendererProperties, rendererIndex=2}
   {rendererClass=&NiagaraSpriteRendererProperties, rendererIndex=3}
  ]
 }
}
summary {
 systemName=NS_Ghosthand
 emitters [
  {emitterName=NE_VC_SingleBoost_Sprite_Parent002, rendererClasses [&NiagaraSpriteRendererProperties]}
  {emitterName=NE_VC_SingleBoost_Ribbon_Parent001, rendererClasses [&NiagaraRibbonRendererProperties]}
 ]
 userVariables [
  User.bIsFrontendPreview {type=NiagaraBool}
  User.bIsPainted {type=NiagaraBool, defaultValue=true}
  User.blsIsCoroColor {type=NiagaraBool}
  User.bPreviewVehicleSpeed {type=NiagaraBool}
  User.ColorVariantID {type=NiagaraInt32}
  User.DTCarOffset {type=NiagaraPosition, description="", defaultValue {}}
  User.GameModeID {type=NiagaraInt32}
  User.PaintedColor {type=LinearColor, description="", defaultValue {struct=LinearColor, value=<rgba 1 1 1 1>}}
  User.PaintedScalar {type=NiagaraFloat, defaultValue=1, description=""}
  User.PortCount {type=NiagaraInt32, defaultValue=1}
 ]
}
