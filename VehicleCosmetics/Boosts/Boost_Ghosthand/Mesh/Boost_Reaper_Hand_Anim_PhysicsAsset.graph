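# Physics asset for the reaper hand skeletal mesh used by the Ghosthand vehicle boost cosmetic.
# It approximates the hand with two capsule bodies, one on the palm joint and one on the wrist joint.
# A single constraint links palm to wrist with equal angular limits on both swing axes and twist, leaving fingers without bodies of their own.
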PhysicsAsset /VehicleCosmetics/Boosts/Boost_Ghosthand/Mesh/Boost_Reaper_Hand_Anim_PhysicsAsset
bodies {
 palm_jnt [{shapeType=Capsule, length=23.6281, radius=14.9822, center=<15.477 1.0608 -1.0693>, rotation=<rot 0 81.4479 -83.0634>}]
 wrist_jnt [{shapeType=Capsule, length=24.4138, radius=9.4033, center=<15.1993 1.0661 8.6244>, rotation=<rot 0 77.6803 -48.3429>}]
}
constraints [{bone1Name=palm_jnt, bone2Name=wrist_jnt, swing1LimitDegrees=45, swing2LimitDegrees=45, twistLimitDegrees=45}]
